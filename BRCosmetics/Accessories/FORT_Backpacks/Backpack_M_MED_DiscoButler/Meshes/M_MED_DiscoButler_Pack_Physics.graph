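# Physics asset for the DiscoButler backpack cosmetic, defining capsule bodies and joints for a hanging fabric piece and a dog keychain.
# Small base bodies for each piece are angularly locked to the large pack capsule and anchor the dynamic chains.
# The fabric chain forks after its third segment into two ends, while the keychain is a linear four-segment chain.
# Chain joints have narrow, asymmetric swing limits and zero-degree twist limits.
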PhysicsAsset /BRCosmetics/Accessories/FORT_Backpacks/Backpack_M_MED_DiscoButler/Meshes/M_MED_DiscoButler_Pack_Physics
bodies {
 base_dogkeychain_0 [{shapeType=Capsule, length=0.505, radius=0.505, shapeName=base_dogkeychain_0_capsule, center=<0 -0.7317 0>, rotation=<rot 0 -90 0>}]
 base_fabric_01 [{shapeType=Capsule, length=1, radius=1, shapeName=base_fabric_01_capsule, center=<0.2428 0 0>, rotation=<rot 0 91.0621 -86.214>}]
 dyn_dogkeychain_01 [{shapeType=Capsule, length=1.9676, radius=4, shapeName=dyn_dogkeychain_01_capsule, center=<4.585 0 -0.9477>, rotation=<rot 76.1188 0 0>}]
 dyn_dogkeychain_02 [{shapeType=Capsule, length=0, radius=4, shapeName=dyn_dogkeychain_02_capsule, center=<5 0 0.2286>, rotation=<rot 90 0 0>}]
 dyn_dogkeychain_03 [{shapeType=Capsule, length=1.1085, radius=4, shapeName=dyn_dogkeychain_03_capsule, center=<5 0 -0.1486>, rotation=<rot 90 0 0>}]
 dyn_dogkeychain_04 [{shapeType=Capsule, length=0, radius=4, shapeName=dyn_dogkeychain_04_capsule, center=<5 0 -0.1486>, rotation=<rot 90 0 0>}]
 dyn_fabric_01 [{shapeType=Capsule, length=4.7392, radius=5, shapeName=dyn_fabric_01_capsule, center=<5.0007 -4.8 0>, rotation=<rot 0 0 20>}]
 dyn_fabric_02 [{shapeType=Capsule, length=5.2712, radius=5, shapeName=dyn_fabric_02_capsule, center=<4.9407 -4 1.8>, rotation=<rot 0 0 25>}]
 dyn_fabric_03 [{shapeType=Capsule, length=5.0801, radius=5, shapeName=dyn_fabric_03_capsule, center=<5 -3 1.8>, rotation=<rot 0 0 30>}]
 dyn_fabric_04 [{shapeType=Capsule, length=0.7294, radius=5, shapeName=dyn_fabric_04_capsule, center=<5 0 0>, rotation=<rot 90 0 0>}]
 dyn_fabric_05 [{shapeType=Capsule, length=2.8555, radius=5, shapeName=dyn_fabric_05_capsule, center=<5 0 0>, rotation=<rot 90 0 0>}]
 pack [{shapeType=Capsule, length=17.2807, radius=10.4331, shapeName=pack_capsule, center=<2.6098 0.6983 -0.6716>, rotation=<rot 0 0 3.3174>}]
}
constraints [
 {bone1Name=dyn_fabric_01, bone2Name=base_fabric_01, swing1LimitDegrees=35, swing2LimitDegrees=15, twistLimitDegrees=0}
 {bone1Name=dyn_fabric_02, bone2Name=dyn_fabric_01, swing1LimitDegrees=35, swing2LimitDegrees=15, twistLimitDegrees=0}
 {bone1Name=dyn_fabric_03, bone2Name=dyn_fabric_02, swing1LimitDegrees=35, swing2LimitDegrees=15, twistLimitDegrees=0}
 {bone1Name=dyn_fabric_04, bone2Name=dyn_fabric_03, swing1LimitDegrees=35, swing2LimitDegrees=15, twistLimitDegrees=0}
 {bone1Name=dyn_fabric_05, bone2Name=dyn_fabric_03, swing1LimitDegrees=35, swing2LimitDegrees=15, twistLimitDegrees=0}
 {bone1Name=dyn_dogkeychain_01, bone2Name=base_dogkeychain_0, swing1LimitDegrees=25, swing2LimitDegrees=17, twistLimitDegrees=0}
 {bone1Name=dyn_dogkeychain_02, bone2Name=dyn_dogkeychain_01, swing1LimitDegrees=30, swing2LimitDegrees=17, twistLimitDegrees=0}
 {bone1Name=dyn_dogkeychain_03, bone2Name=dyn_dogkeychain_02, swing1LimitDegrees=30, swing2LimitDegrees=17, twistLimitDegrees=0}
 {bone1Name=dyn_dogkeychain_04, bone2Name=dyn_dogkeychain_03, swing1LimitDegrees=30, swing2LimitDegrees=17, twistLimitDegrees=0}
 {
  bone1Name=base_fabric_01
  bone2Name=pack
  swing1LimitDegrees=45
  swing1Motion=Locked
  swing2LimitDegrees=45
  swing2Motion=Locked
  twistLimitDegrees=45
  twistMotion=Locked
 }
 {
  bone1Name=base_dogkeychain_0
  bone2Name=pack
  swing1LimitDegrees=45
  swing1Motion=Locked
  swing2LimitDegrees=45
  swing2Motion=Locked
  twistLimitDegrees=45
  twistMotion=Locked
 }
]
